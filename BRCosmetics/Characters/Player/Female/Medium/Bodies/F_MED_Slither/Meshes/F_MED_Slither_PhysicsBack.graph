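# Physics asset for the back skirt of the F_MED_Slither female player body, covering center, left and right bone chains.
# Each chain runs from a small base capsule through two larger capsules, with tight swing and twist limit values between links.
# Bases join the pelvis with free rotation, as do the thigh capsules that provide leg volumes.
# Free-rotation constraints also cross-link the chain tips side to side, tying the three panels together.
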PhysicsAsset /BRCosmetics/Characters/Player/Female/Medium/Bodies/F_MED_Slither/Meshes/F_MED_Slither_PhysicsBack
bodies {
 dyn_skirt_bk_01 [{shapeType=Capsule, length=6.75, radius=5, shapeName=dyn_skirt_bk_01_capsule, center=<5 0 5.2>, rotation=<rot 0 0 89.9999>}]
 dyn_skirt_bk_02 [{shapeType=Capsule, length=6.25, radius=5.5, shapeName=dyn_skirt_bk_02_capsule, center=<7 0 3.5>, rotation=<rot 0 0 89.9999>}]
 dyn_skirt_bk_base [{shapeType=Capsule, length=0.505, radius=0.505, shapeName=dyn_skirt_bk_base_capsule, rotation=<rot 0 -90 0>}]
 dyn_skirt_bk_l_01 [{shapeType=Capsule, length=12.75, radius=5, shapeName=dyn_skirt_bk_l_01_capsule, center=<5 0.0001 5.2>, rotation=<rot 0 0 89.9999>}]
 dyn_skirt_bk_l_02 [{shapeType=Capsule, length=12.25, radius=5.5, shapeName=dyn_skirt_bk_l_02_capsule, center=<7 0 3.75>, rotation=<rot 0 0 89.9999>}]
 dyn_skirt_bk_l_base [{shapeType=Capsule, length=0.505, radius=0.505, shapeName=dyn_skirt_bk_l_base_capsule, rotation=<rot 0 -90 0>}]
 dyn_skirt_bk_r_01 [{shapeType=Capsule, length=12.75, radius=5, shapeName=dyn_skirt_bk_r_01_capsule, center=<-5 0 -5.25>, rotation=<rot 0 0 89.9999>}]
 dyn_skirt_bk_r_02 [{shapeType=Capsule, length=12.75, radius=5, shapeName=dyn_skirt_bk_r_02_capsule, center=<-7 -0.0001 -3.8>, rotation=<rot 0 0 89.9999>}]
 dyn_skirt_bk_r_base [{shapeType=Capsule, length=0.505, radius=0.505, shapeName=dyn_skirt_bk_r_base_capsule, rotation=<rot 0 -90 0>}]
 pelvis [{shapeType=Capsule, length=3, radius=13.25, shapeName=pelvis_capsule}]
 thigh_l [{
   shapeType=Capsule
   length=25.6815
   radius=10.3553
   shapeName=thigh_l_capsule
   center=<-11.0601 -2.4673 1.6943>
   rotation=<rot 0 -86.8851 -99.2367>
  }]
 thigh_r [{
   shapeType=Capsule
   length=25.9315
   radius=10.1053
   shapeName=thigh_l_capsule
   center=<11.3382 2.4825 -4.0801>
   rotation=<rot 0 -86.8852 -89.2367>
  }]
}
constraints [
 {bone1Name=dyn_skirt_bk_02, bone2Name=dyn_skirt_bk_01, swing1LimitDegrees=10, swing2LimitDegrees=60, twistLimitDegrees=5}
 {bone1Name=dyn_skirt_bk_l_02, bone2Name=dyn_skirt_bk_l_01, swing1LimitDegrees=10, swing2LimitDegrees=60, twistLimitDegrees=5}
 {bone1Name=dyn_skirt_bk_r_02, bone2Name=dyn_skirt_bk_r_01, swing1LimitDegrees=10, swing2LimitDegrees=60, twistLimitDegrees=5}
 {bone1Name=dyn_skirt_bk_r_01, bone2Name=dyn_skirt_bk_r_base, swing1LimitDegrees=10, swing2LimitDegrees=60, twistLimitDegrees=5}
 {bone1Name=dyn_skirt_bk_01, bone2Name=dyn_skirt_bk_base, swing1LimitDegrees=10, swing2LimitDegrees=60, twistLimitDegrees=5}
 {bone1Name=dyn_skirt_bk_l_01, bone2Name=dyn_skirt_bk_l_base, swing1LimitDegrees=10, swing2LimitDegrees=60, twistLimitDegrees=5}
 {
  bone1Name=dyn_skirt_bk_l_base
  bone2Name=pelvis
  swing1LimitDegrees=45
  swing1Motion=Free
  swing2LimitDegrees=45
  swing2Motion=Free
  twistLimitDegrees=45
  twistMotion=Free
 }
 {
  bone1Name=dyn_skirt_bk_base
  bone2Name=pelvis
  swing1LimitDegrees=45
  swing1Motion=Free
  swing2LimitDegrees=45
  swing2Motion=Free
  twistLimitDegrees=45
  twistMotion=Free
 }
 {
  bone1Name=dyn_skirt_bk_r_base
  bone2Name=pelvis
  swing1LimitDegrees=45
  swing1Motion=Free
  swing2LimitDegrees=45
  swing2Motion=Free
  twistLimitDegrees=45
  twistMotion=Free
 }
 {
  bone1Name=thigh_l
  bone2Name=pelvis
  swing1LimitDegrees=45
  swing1Motion=Free
  swing2LimitDegrees=45
  swing2Motion=Free
  twistLimitDegrees=45
  twistMotion=Free
 }
 {
  bone1Name=thigh_r
  bone2Name=pelvis
  swing1LimitDegrees=45
  swing1Motion=Free
  swing2LimitDegrees=45
  swing2Motion=Free
  twistLimitDegrees=45
  twistMotion=Free
 }
 {
  bone1Name=dyn_skirt_bk_02
  bone2Name=dyn_skirt_bk_r_02
  swing1LimitDegrees=45
  swing1Motion=Free
  swing2LimitDegrees=45
  swing2Motion=Free
  twistLimitDegrees=45
  twistMotion=Free
 }
 {
  bone1Name=dyn_skirt_bk_l_02
  bone2Name=dyn_skirt_bk_02
  swing1LimitDegrees=45
  swing1Motion=Free
  swing2LimitDegrees=45
  swing2Motion=Free
  twistLimitDegrees=45
  twistMotion=Free
 }
]
