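# Physics asset for the F_MED_Slither female cosmetic body, setting up collision and joints for its dangling belt.
# Two capsule chains, four links on the left and five on the right, branch from beltMain, each joint with limited swing and tight twist.
# Pelvis and thigh capsules provide body shapes around the belt.
# The beltMain-to-pelvis and thigh-to-pelvis constraints leave swing and twist free, so only the chain joints have angular limits.
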PhysicsAsset /BRCosmetics/Characters/Player/Female/Medium/Bodies/F_MED_Slither/Meshes/F_MED_Slither_PhysicsBelt
bodies {
 beltMain [{shapeType=Capsule, length=3.8968, radius=2.4116, shapeName=beltMain_capsule, rotation=<rot 90 0 0>}]
 belt_dyn_l_1 [{shapeType=Capsule, length=2, radius=2, shapeName=belt_dyn_l_1_capsule, center=<2 0 0>, rotation=<rot 90 0 0>}]
 belt_dyn_l_2 [{shapeType=Capsule, length=2, radius=2, shapeName=belt_dyn_l_2_capsule, center=<2 0 0>, rotation=<rot 90 0 0>}]
 belt_dyn_l_3 [{shapeType=Capsule, length=2, radius=2, shapeName=belt_dyn_l_3_capsule, center=<2 0 0>, rotation=<rot 90 0 0>}]
 belt_dyn_l_4 [{shapeType=Capsule, length=2, radius=2, shapeName=belt_dyn_l_4_capsule, center=<2 0 0>, rotation=<rot 90 0 0>}]
 belt_dyn_r_1 [{shapeType=Capsule, length=2, radius=2, shapeName=belt_dyn_r_1_capsule, center=<-2 0 0>, rotation=<rot 90 0 0>}]
 belt_dyn_r_2 [{shapeType=Capsule, length=2, radius=2, shapeName=belt_dyn_r_2_capsule, center=<-2 0 0>, rotation=<rot 90 0 0>}]
 belt_dyn_r_3 [{shapeType=Capsule, length=2, radius=2, shapeName=belt_dyn_r_3_capsule, center=<-2 0 0>, rotation=<rot 90 0 0>}]
 belt_dyn_r_4 [{shapeType=Capsule, length=2, radius=2, shapeName=belt_dyn_r_4_capsule, center=<-2 0 0>, rotation=<rot 90 0 0>}]
 belt_dyn_r_5 [{shapeType=Capsule, length=2, radius=2, shapeName=belt_dyn_r_5_capsule, center=<-2 0 0>, rotation=<rot 90 0 0>}]
 pelvis [
  {shapeType=Capsule, length=13.7925, radius=8.9339, shapeName=pelvis_capsule, center=<-9.5095 1.6307 0>, rotation=<rot -90 0 0>}
  {shapeType=Capsule, length=14.25, radius=8, shapeName=pelvis_capsule, center=<13.768 0.0001 0>, rotation=<rot -0.0023 0.0001 0>}
 ]
 thigh_l [{
   shapeType=Capsule
   length=20.9315
   radius=9.6053
   shapeName=thigh_l_capsule
   center=<-7.9283 -0.5656 1.0426>
   rotation=<rot 0.0665 -91.8847 -89.2396>
  }]
 thigh_r [{
   shapeType=Capsule
   length=19.9894
   radius=10.6556
   shapeName=thigh_r_capsule
   center=<8.0713 1.3397 -2.0568>
   rotation=<rot 0.0574 88.9578 -89.3441>
  }]
}
constraints [
 {bone1Name=belt_dyn_r_1, bone2Name=beltMain, swing1LimitDegrees=45, swing2LimitDegrees=20, twistLimitDegrees=5}
 {bone1Name=belt_dyn_r_2, bone2Name=belt_dyn_r_1, swing1LimitDegrees=45, swing2LimitDegrees=20, twistLimitDegrees=5}
 {bone1Name=belt_dyn_r_3, bone2Name=belt_dyn_r_2, swing1LimitDegrees=45, swing2LimitDegrees=20, twistLimitDegrees=5}
 {bone1Name=belt_dyn_r_4, bone2Name=belt_dyn_r_3, swing1LimitDegrees=45, swing2LimitDegrees=20, twistLimitDegrees=5}
 {bone1Name=belt_dyn_r_5, bone2Name=belt_dyn_r_4, swing1LimitDegrees=45, swing2LimitDegrees=20, twistLimitDegrees=5}
 {bone1Name=belt_dyn_l_1, bone2Name=beltMain, swing1LimitDegrees=45, swing2LimitDegrees=20, twistLimitDegrees=5}
 {bone1Name=belt_dyn_l_2, bone2Name=belt_dyn_l_1, swing1LimitDegrees=45, swing2LimitDegrees=20, twistLimitDegrees=5}
 {bone1Name=belt_dyn_l_3, bone2Name=belt_dyn_l_2, swing1LimitDegrees=45, swing2LimitDegrees=20, twistLimitDegrees=5}
 {bone1Name=belt_dyn_l_4, bone2Name=belt_dyn_l_3, swing1LimitDegrees=45, swing2LimitDegrees=20, twistLimitDegrees=5}
 {
  bone1Name=beltMain
  bone2Name=pelvis
  swing1LimitDegrees=45
  swing1Motion=Free
  swing2LimitDegrees=45
  swing2Motion=Free
  twistLimitDegrees=45
  twistMotion=Free
 }
 {
  bone1Name=thigh_l
  bone2Name=pelvis
  swing1LimitDegrees=45
  swing1Motion=Free
  swing2LimitDegrees=45
  swing2Motion=Free
  twistLimitDegrees=45
  twistMotion=Free
 }
 {
  bone1Name=thigh_r
  bone2Name=pelvis
  swing1LimitDegrees=45
  swing1Motion=Free
  swing2LimitDegrees=45
  swing2Motion=Free
  twistLimitDegrees=45
  twistMotion=Free
 }
]
