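# Physics asset for the right arm cuff of the F_MED_Slither female medium player body.
# A tiny capsule on lowerarm_r anchors a larger, offset capsule on the lowerarm_r_dyn bone.
# One constraint ties the two, allowing very slight swing and a zero twist limit, so the cuff has little room to move.
# The dynamic capsule's shape name refers to the left arm, suggesting a copy from that side.
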PhysicsAsset /Game/Characters/Player/Female/Medium/Bodies/F_MED_Slither/Meshes/F_MED_Slither_PhysicsArmCuffRt
bodies {
 lowerarm_r [{shapeType=Capsule, length=0.505, radius=0.505, shapeName=lowerarm_r_capsule, rotation=<rot 0 -90 0>}]
 lowerarm_r_dyn [{shapeType=Capsule, length=5, radius=2, shapeName=lowerarm_l_dyn_capsule, center=<-5 0 0>, rotation=<rot 90 0 0>}]
}
constraints [{bone1Name=lowerarm_r_dyn, bone2Name=lowerarm_r, swing1LimitDegrees=1.5, swing2LimitDegrees=1.5, twistLimitDegrees=0}]
