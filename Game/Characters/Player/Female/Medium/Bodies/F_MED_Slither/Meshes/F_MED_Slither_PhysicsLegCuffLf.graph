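# Physics asset for the left leg cuff on the Slither female medium player body, apparently for secondary motion of that piece.
# It holds two capsule bodies: a wider one on calf_l and a thinner one on the calf_l_dyn bone.
# One constraint joins calf_l_dyn to calf_l with a narrow limit on both swing axes and a zero twist limit, allowing only a slight wobble relative to the calf.
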PhysicsAsset /Game/Characters/Player/Female/Medium/Bodies/F_MED_Slither/Meshes/F_MED_Slither_PhysicsLegCuffLf
bodies {
 calf_l [{
   shapeType=Capsule
   length=7.0027
   radius=6.4561
   shapeName=calf_l_capsule
   center=<-4.4404 -1.6085 -0.1211>
   rotation=<rot 88.913 110.4552 -159.5476>
  }]
 calf_l_dyn [{shapeType=Capsule, length=5, radius=2, shapeName=calf_l_dyn_capsule, center=<-5 0 0>, rotation=<rot 90 0 0>}]
}
constraints [{bone1Name=calf_l_dyn, bone2Name=calf_l, swing1LimitDegrees=4, swing2LimitDegrees=4, twistLimitDegrees=0}]
