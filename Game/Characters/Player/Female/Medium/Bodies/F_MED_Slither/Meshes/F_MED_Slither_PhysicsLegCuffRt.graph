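# Physics asset for the right leg cuff of the F_MED_Slither female player body, covering the right calf and its dynamic bone.
# A larger capsule on calf_r represents the leg, and a small capsule sits on calf_r_dyn.
# One constraint ties calf_r_dyn to calf_r with narrow swing limits and a zero twist limit.
# The dynamic capsule's shape name says calf_l although it belongs to the right-side bone.
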PhysicsAsset /Game/Characters/Player/Female/Medium/Bodies/F_MED_Slither/Meshes/F_MED_Slither_PhysicsLegCuffRt
bodies {
 calf_r [{
   shapeType=Capsule
   length=6.8358
   radius=6.3375
   shapeName=calf_r_capsule
   center=<4.1665 1.6518 0.0867>
   rotation=<rot 79.306 122.8278 -147.6278>
  }]
 calf_r_dyn [{shapeType=Capsule, length=5, radius=2, shapeName=calf_l_dyn_capsule, center=<5 0 0>, rotation=<rot 90 0 0>}]
}
constraints [{bone1Name=calf_r_dyn, bone2Name=calf_r, swing1LimitDegrees=4, swing2LimitDegrees=4, twistLimitDegrees=0}]
