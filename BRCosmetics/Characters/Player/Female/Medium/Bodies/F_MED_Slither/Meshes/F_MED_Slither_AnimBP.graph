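# Animation Blueprint for the F_MED_Slither player cosmetic body, adding secondary motion and corrective deformation to the pose copied from the parent mesh.
# Two borrowed Stereo Control Rigs feed weighted back-skirt constraints, then RigidBody sims for back and belt.
# Bone offsets finish collar, shoulders and skirt, driven by bone-delta alphas and skydiving state from the EventGraph.
# The cuff RigidBody chain is disconnected; ShoulderRt is unused, as both clavicles read ShoulderLf.
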AnimBlueprint /BRCosmetics/Characters/Player/Female/Medium/Bodies/F_MED_Slither/Meshes/F_MED_Slither_AnimBP
target_skeleton=&/F_MED_Slither_Skeleton
graphs [
 AnimGraph {nodes [
   n1 = RigidBody() {
    bClampLinearTranslationLimitToRefPose=true
    bOverrideWorldGravity=true
    overridePhysicsAsset=/Game/Characters/Player/Female/Medium/Bodies/F_MED_Slither/Meshes/F_MED_Slither_PhysicsLegCuffLf
    simulationSpace=BaseBoneSpace
    alphaBoolBlend {blendOption=Linear}
    baseBoneRef {boneName=calf_l}
    componentAppliedLinearAccClamp=<3000 3000 3000>
    componentLinearAccScale=<0.4 0.3 0.3>
    componentLinearVelScale=<0.4 0.3 0.3>
    overrideWorldGravity {z=-980}
   }
   n2 = RigidBody(ComponentPose=@n1) {
    bClampLinearTranslationLimitToRefPose=true
    bOverrideWorldGravity=true
    overridePhysicsAsset=/Game/Characters/Player/Female/Medium/Bodies/F_MED_Slither/Meshes/F_MED_Slither_PhysicsLegCuffRt
    simulationSpace=BaseBoneSpace
    alphaBoolBlend {blendOption=Linear}
    baseBoneRef {boneName=calf_r}
    componentAppliedLinearAccClamp=<3000 3000 3000>
    componentLinearAccScale=<0.4 0.3 0.3>
    componentLinearVelScale=<0.4 0.3 0.3>
    overrideWorldGravity {z=-980}
   }
   n3 = RigidBody(ComponentPose=@n2) {
    bClampLinearTranslationLimitToRefPose=true
    bOverrideWorldGravity=true
    overridePhysicsAsset=/Game/Characters/Player/Female/Medium/Bodies/F_MED_Slither/Meshes/F_MED_Slither_PhysicsArmCuffLf
    simulationSpace=BaseBoneSpace
    alphaBoolBlend {blendOption=Linear}
    baseBoneRef {boneName=lowerarm_l}
    componentAppliedLinearAccClamp=<3000 3000 3000>
    componentLinearAccScale=<0.4 0.3 0.3>
    componentLinearVelScale=<0.4 0.3 0.3>
    overrideWorldGravity {z=-980}
   }
   RigidBody(ComponentPose=@n3) {
    bClampLinearTranslationLimitToRefPose=true
    bOverrideWorldGravity=true
    overridePhysicsAsset=/Game/Characters/Player/Female/Medium/Bodies/F_MED_Slither/Meshes/F_MED_Slither_PhysicsArmCuffRt
    simulationSpace=BaseBoneSpace
    alphaBoolBlend {blendOption=Linear}
    baseBoneRef {boneName=lowerarm_r}
    componentAppliedLinearAccClamp=<3000 3000 3000>
    componentLinearAccScale=<0.4 0.3 0.3>
    componentLinearVelScale=<0.4 0.3 0.3>
    overrideWorldGravity {z=-980}
   }
   n4 = Control Rig(Source=Copy Pose From Mesh() {bUseAttachedParent=true}) {
    alphaBoolBlend {blendOption=Linear}
    controlRigAssetReference {blueprintRigClass=/BRCosmetics/Characters/Player/Female/Medium/Bodies/F_MED_Stereo/Meshes/F_MED_Stereo_Deform_CtrlRig_twists.F_MED_Stereo_Deform_CtrlRig_twists_C}
   }
   n5 = Control Rig!(
    Source=@n4
    Alpha
    KneeRot_fr_l_114deg:Vector
    KneeTrans_fr_l_114deg:Vector="0.258,-3.591,0.009"
    KneeScale_fr_l_114deg:Vector="1,1,1"
    KneeScale_fr_l_130deg:Vector="1,1,1"
    KneeRot_fr_l_130deg:Vector="0,0,-51.2"
    KneeTrans_fr_l_130deg:Vector="0.258,-3.591,0.009"
   ) -> (Pose) {
    alphaBoolBlend {blendOption=Linear}
    controlRigAssetReference {blueprintRigClass=/BRCosmetics/Characters/Player/Female/Medium/Bodies/F_MED_Stereo/Meshes/F_MED_Stereo_Deform_CtrlRig_sculpters.F_MED_Stereo_Deform_CtrlRig_sculpters_C}
   }
   n6 = "Constraint - dyn_skirt_bk_l_01 (PP)"!(ComponentPose=Local To Component(LocalPose=@n5), Alpha=1, ConstraintWeights_0=0.5, ConstraintWeights_1:float=0.5) -> (Pose) {
    alphaBoolBlend {blendOption=Linear}
    boneToModify {boneName=dyn_skirt_bk_l_01}
    constraintSetup [{targetBone {boneName=deform_thigh_bk_l}}, {targetBone {boneName=pelvis}}]
    constraintWeights [1,1]
   }
   n7 = "Constraint - dyn_skirt_bk_r_01 (PP)"!(ComponentPose=@n6, Alpha=1, ConstraintWeights_0=0.5, ConstraintWeights_1:float=0.5) -> (Pose) {
    alphaBoolBlend {blendOption=Linear}
    boneToModify {boneName=dyn_skirt_bk_r_01}
    constraintSetup [{targetBone {boneName=deform_thigh_bk_r}}, {targetBone {boneName=pelvis}}]
    constraintWeights [1,1]
   }
   n8 = "Constraint - dyn_skirt_bk_l_02 (PPPPPPPP)"!(
    ComponentPose=@n7
    Alpha=1
    ConstraintWeights_0=0.6918
    ConstraintWeights_1:float=0.1242
    ConstraintWeights_2:float=0.0738
    ConstraintWeights_3:float=0.0358
    ConstraintWeights_4:float=0.0314
    ConstraintWeights_5:float=0.0228
    ConstraintWeights_6:float=0.0181
    ConstraintWeights_7:float=0.0021
   ) -> (Pose) {
    alphaBoolBlend {blendOption=Linear}
    boneToModify {boneName=dyn_skirt_bk_l_02}
    constraintSetup [
     {targetBone {boneName=pelvis}}
     {targetBone {boneName=thigh_twist_01_l}}
     {targetBone {boneName=deform_thigh_bk_l}}
     {targetBone {boneName=spine_02}}
     {targetBone {boneName=spine_01}}
     {targetBone {boneName=thigh_l}}
     {targetBone {boneName=deform_thigh_fr_l}}
     {targetBone {boneName=deform_thigh_bk_r}}
    ]
    constraintWeights [1,1,1,1,1,1,1,1]
   }
   n9 = "Constraint - dyn_skirt_bk_r_02 (PPPPPPPP)"!(
    ComponentPose=@n8
    Alpha=1
    ConstraintWeights_0=0.6919
    ConstraintWeights_1:float=0.1241
    ConstraintWeights_2:float=0.0738
    ConstraintWeights_3:float=0.0359
    ConstraintWeights_4:float=0.0315
    ConstraintWeights_5:float=0.0227
    ConstraintWeights_6:float=0.018
    ConstraintWeights_7:float=0.0021
   ) -> (Pose) {
    alphaBoolBlend {blendOption=Linear}
    boneToModify {boneName=dyn_skirt_bk_r_02}
    constraintSetup [
     {targetBone {boneName=pelvis}}
     {targetBone {boneName=thigh_twist_01_r}}
     {targetBone {boneName=deform_thigh_bk_r}}
     {targetBone {boneName=spine_02}}
     {targetBone {boneName=spine_01}}
     {targetBone {boneName=thigh_r}}
     {targetBone {boneName=deform_thigh_fr_r}}
     {targetBone {boneName=deform_thigh_bk_l}}
    ]
    constraintWeights [1,1,1,1,1,1,1,1]
   }
   n10 = "Constraint - dyn_skirt_bk_01 (PP)"!(ComponentPose=@n9, Alpha=1, ConstraintWeights_0=0.5, ConstraintWeights_1:float=0.5) -> (Pose) {
    alphaBoolBlend {blendOption=Linear}
    boneToModify {boneName=dyn_skirt_bk_01}
    constraintSetup [{targetBone {boneName=dyn_skirt_bk_l_01}}, {targetBone {boneName=dyn_skirt_bk_r_01}}]
    constraintWeights [1,1]
   }
   n11 = "Constraint - dyn_skirt_bk_02 (PP)"!(ComponentPose=@n10, Alpha=1, ConstraintWeights_0=0.5, ConstraintWeights_1:float=0.5) -> (Pose) {
    alphaBoolBlend {blendOption=Linear}
    boneToModify {boneName=dyn_skirt_bk_02}
    constraintSetup [{targetBone {boneName=dyn_skirt_bk_l_02}}, {targetBone {boneName=dyn_skirt_bk_r_02}}]
    constraintWeights [1,1]
   }
   n12 = RigidBody(ComponentPose=@n11) {
    bClampLinearTranslationLimitToRefPose=true
    bOverrideWorldGravity=true
    overridePhysicsAsset=&/F_MED_Slither_PhysicsBack
    simulationSpace=BaseBoneSpace
    alphaBoolBlend {blendOption=Linear}
    baseBoneRef {boneName=pelvis}
    componentAppliedLinearAccClamp=<3000 3000 3000>
    componentLinearAccScale=<0.45 0.35 0.35>
    componentLinearVelScale=<0.45 0.35 0.35>
    overrideWorldGravity {z=-980}
   }
   n13 = RigidBody(ComponentPose=@n12) {
    bClampLinearTranslationLimitToRefPose=true
    bOverrideWorldGravity=true
    overridePhysicsAsset=&/F_MED_Slither_PhysicsBelt
    simulationSpace=BaseBoneSpace
    alphaBoolBlend {blendOption=Linear}
    baseBoneRef {boneName=pelvis}
    componentAppliedLinearAccClamp=<3000 3000 3000>
    componentLinearAccScale=<0.4 0.3 0.3>
    componentLinearVelScale=<0.4 0.3 0.3>
    overrideWorldGravity {z=-980}
   }
   n14 = "Transform (Modify) Bone - Bone: neck_collar"(ComponentPose=@n13, Alpha=Get HeadRot() -> (HeadRot:double), Translation="0,0,0", Rotation="0,0,0", Scale="1,7,1") {
    rotationMode=BMM_Additive
    scaleMode=BMM_Additive
    translationMode=BMM_Additive
    alphaBoolBlend {blendOption=Linear}
    alphaScaleBiasClamp {bInterpResult=true, interpSpeedDecreasing=1, interpSpeedIncreasing=1}
    boneToModify {boneName=neck_collar}
    scale {y=7}
   }
   n15 = "Transform (Modify) Bone - Bone: neck_collar"(ComponentPose=@n14, Alpha=Get isSkydiving() -> (isSkydiving:double), Translation="0,-2,5", Rotation="0,0,0", Scale="3,1,10") {
    rotationMode=BMM_Additive
    translationMode=BMM_Additive
    alphaBoolBlend {blendOption=Linear}
    alphaScaleBiasClamp {bInterpResult=true, interpSpeedDecreasing=1, interpSpeedIncreasing=1}
    boneToModify {boneName=neck_collar}
    scale {x=3, z=10}
    translation {y=-2, z=5}
   }
   n16 = "Transform (Modify) Bone - Bone: spine_05_Cheat"(ComponentPose=@n15, Alpha=Get isSkydiving(), Translation="0,0,0", Rotation="0,0,0", Scale="1,1.05,1") {
    rotationMode=BMM_Additive
    scaleMode=BMM_Additive
    translationMode=BMM_Additive
    alphaBoolBlend {blendOption=Linear}
    alphaScaleBiasClamp {bInterpResult=true, interpSpeedDecreasing=1, interpSpeedIncreasing=1}
    boneToModify {boneName=spine_05_Cheat}
    scale {y=1.05}
   }
   n17 = Get isSkydiving()
   n18 = "Transform (Modify) Bone - Bone: dyn_skirt_bk_l_MSTR"(ComponentPose=@n16, Alpha=@n17, Translation="0,0,4", Rotation="0,0,0", Scale="1,1,1") {
    rotationMode=BMM_Additive
    translationMode=BMM_Additive
    alphaBoolBlend {blendOption=Linear}
    alphaScaleBiasClamp {bInterpResult=true, interpSpeedDecreasing=1, interpSpeedIncreasing=1}
    boneToModify {boneName=dyn_skirt_bk_l_MSTR}
    translation {z=4}
   }
   n19 = "Transform (Modify) Bone - Bone: dyn_skirt_bk_MSTR"(ComponentPose=@n18, Alpha=@n17, Translation="0,0,4", Rotation="0,0,0", Scale="1,1,1") {
    rotationMode=BMM_Additive
    translationMode=BMM_Additive
    alphaBoolBlend {blendOption=Linear}
    alphaScaleBiasClamp {bInterpResult=true, interpSpeedDecreasing=1, interpSpeedIncreasing=1}
    boneToModify {boneName=dyn_skirt_bk_MSTR}
    translation {z=4}
   }
   n20 = "Transform (Modify) Bone - Bone: dyn_skirt_bk_r_MSTR"(ComponentPose=@n19, Alpha=@n17, Translation="0,0,4", Rotation="0,0,0", Scale="1,1,1") {
    rotationMode=BMM_Additive
    translationMode=BMM_Additive
    alphaBoolBlend {blendOption=Linear}
    alphaScaleBiasClamp {bInterpResult=true, interpSpeedDecreasing=1, interpSpeedIncreasing=1}
    boneToModify {boneName=dyn_skirt_bk_r_MSTR}
    translation {z=4}
   }
   n21 = "Transform (Modify) Bone - Bone: clavicle_l_Cheat"(ComponentPose=@n20, Alpha=Get ShoulderLf() -> (ShoulderLf:double), Translation="0,0,4", Rotation="0,0,0", Scale="1,2,1") {
    rotationMode=BMM_Additive
    translationMode=BMM_Additive
    alphaBoolBlend {blendOption=Linear}
    alphaScaleBiasClamp {bInterpResult=true, interpSpeedDecreasing=1, interpSpeedIncreasing=1}
    boneToModify {boneName=clavicle_l_Cheat}
    scale {y=2}
    translation {z=4}
   }
   n22 = "Transform (Modify) Bone - Bone: clavicle_r_Cheat"(ComponentPose=@n21, Alpha=Get ShoulderLf(), Translation="0,0,4", Rotation="0,0,0", Scale="1,1,1") {
    rotationMode=BMM_Additive
    translationMode=BMM_Additive
    alphaBoolBlend {blendOption=Linear}
    alphaScaleBiasClamp {bInterpResult=true, interpSpeedDecreasing=1, interpSpeedIncreasing=1}
    boneToModify {boneName=clavicle_r_Cheat}
    translation {z=4}
   }
   n23 = "Transform (Modify) Bone - Bone: neck_collar"(ComponentPose=@n22, Alpha=1, Translation="0,-1.5,1", Rotation="0,0,0", Scale="1,2,1") {
    rotationMode=BMM_Additive
    translationMode=BMM_Additive
    alphaBoolBlend {blendOption=Linear}
    alphaScaleBiasClamp {bInterpResult=true, interpSpeedDecreasing=1, interpSpeedIncreasing=1}
    boneToModify {boneName=neck_collar}
    scale {y=2}
    translation {y=-1.5, z=1}
   }
   n24 = "Transform (Modify) Bone - Bone: clavicle_l_frontendCollar"(ComponentPose=@n23, Alpha=Get alpha collar() -> (alpha collar:float), Translation="15,0,3", Rotation="0,0,0", Scale="1,1,1.1") {
    rotationMode=BMM_Additive
    scaleMode=BMM_Additive
    translationMode=BMM_Additive
    translationSpace=BCS_BoneSpace
    alphaBoolBlend {blendOption=Linear}
    alphaScaleBiasClamp {bInterpResult=true, interpSpeedDecreasing=1, interpSpeedIncreasing=1}
    boneToModify {boneName=clavicle_l_frontendCollar}
    scale {z=1.1}
    translation {x=15, z=3}
   }
   Output Pose(Result=Component To Local(ComponentPose=@n24)) {layerGroup=None}
  ]}
 EventGraph {nodes [
   Event BlueprintUpdateAnimation()
   n1 = GetOwnerSkeletalMeshComponent()
   > Set skeletal mesh(skeletal mesh:Skeletal Mesh Component Object Reference=@n1) -> (Output_Get:Skeletal Mesh Component Object Reference)
   > Set HeadRot(HeadRot:double=GetAlphaFromReferenceTransformDelta(SkeletalMeshComponent=@n1, BoneName=head, BaseName=neck_01, InputRange=(X=0,Y=-100), TwistAxis="1,0,0")) -> (Output_Get:double)
   > Set isSkydiving(isSkydiving:double="To Float (Boolean)"(InBool=Get bIsSkyDiving())) -> (Output_Get:double)
   n2 = Get bIsCrouchSprinting()
   n3 = Get bIsTargeting()
   > Set isADSCrouchWalking(isADSCrouchWalking:double="To Float (Boolean)"(InBool=AND Boolean(A=@n2, B=@n3))) -> (Output_Get:double)
   > Set ShoulderLf(ShoulderLf:double=GetAlphaFromReferenceTransformDelta(SkeletalMeshComponent=@n1, BoneName=clavicle_l, BaseName=spine_05, DeltaComponentType=EulerY, InputRange=(X=0,Y=100), TwistAxis="1,0,0")) -> (Output_Get:double)
   > Set ShoulderRt(ShoulderRt:double=GetAlphaFromReferenceTransformDelta(SkeletalMeshComponent=@n1, BoneName=clavicle_r, BaseName=spine_05, DeltaComponentType=EulerY, InputRange=(X=0,Y=100), TwistAxis="1,0,0")) -> (Output_Get:double)
   > Set alpha collar(alpha collar:float=GetAlphaFromReferenceTransformDelta(
     SkeletalMeshComponent=Get skeletal mesh() -> (skeletal mesh:Skeletal Mesh Component Object Reference)
     BoneName=spine_05
     BaseName=clavicle_l
     DeltaComponentType=EulerY
     InputRange=(X=-10,Y=-40)
     TwistAxis="1,0,0"
    )) -> (Output_Get:float)
  ]}
]
variables [
 alpha collar {type=number}
 HeadRot {type=number}
 isADSCrouchWalking {type=number}
 isSkydiving {type=number}
 ShoulderLf {type=number}
 ShoulderRt {type=number}
 skeletal mesh {type=SkeletalMeshComponent}
]
